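# EQS goal-selection query for enemy spawn-spot defender AI pawns, choosing between perceived enemies and fallback circle goals.
# A composite generator supplies both goal types; enemies are scored by closeness, with a perception test weighing senses including hearing.
# Escort-assignment circle goals are filtered out near the pawn or nearby friends to break up clumps, with a small dot score favoring goals ahead.
# Enemies too far from the spawn location are filtered out as an anti-kiting leash.
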EnvQuery /Game/Athena/AI/GoalSelection/Pawn/EQS_GoalSelection_Pawn_Enemy_SpawnSpotActorDefender_Default
options [{
  generator {
   name=EnvQueryGenerator_Composite_0
   class=EnvQueryGenerator_Composite
   properties {
    optionName="Perceived Enemies && Fallback Circle"
    generators [
     &self:EnvQueryGenerator_Composite_0.FortQueryGenerator_Enemies_1
     &self:EnvQueryGenerator_Composite_0.FortQueryGenerator_GoalOnCircle_0
    ]
   }
  }
  tests [
   FortQueryTest_GoalDistance_0 {
    class=FortQueryTest_GoalDistance
    properties {
     bScoreEncounterGoals=true
     bScoreEnemies=true
     bScoreSpecificAssignments=false
     bScoreWorldGoals=false
     clampMaxType=SpecifiedValue
     clampMinType=SpecifiedValue
     filterType=Range
     floatValueMax=0
     scoreClampMax=10000
     scoreClampMin=200
     scoringEquation=InverseLinear
     scoringFactor=0.5
     testComment=Prefer closer things.
     testPurpose=Score
     assignmentIds []
     prohibitedAssignmentIds [{assignmentTeam=Monster, assignmentType=World, assignmentGameplayTags {gameplayTags [{tagName=NPC.Assignment.Escort}]}}]
    }
   }
   FortQueryTest_PerceptionAll_0 {
    class=FortQueryTest_PerceptionAll
    properties {
     clampMaxType=SpecifiedValue
     clampMinType=SpecifiedValue
     maxSenseAge=21
     minSenseAge=20
     scoreClampMax=1
     scoringFactor=1
     additionalSenseScores {&Class'&AthenaAISense_Hearing'=0.1}
     senseScores [0,0.1,1,0,0,0]
    }
   }
   FortQueryTest_GoalDistance_1 {
    class=FortQueryTest_GoalDistance
    properties {
     bScoreEncounterGoals=false
     bScoreEnemies=false
     bScoreSpecificAssignments=true
     bScoreWorldGoals=false
     clampMaxType=None
     clampMinType=None
     filterType=Minimum
     floatValueMax=0
     floatValueMin=200
     scoreClampMax=0
     scoreClampMin=0
     scoringEquation=Linear
     scoringFactor=1
     testComment=Goals On Circle - Throw out things we're on top of.
     testPurpose=Filter
     assignmentIds [{assignmentTeam=Monster, assignmentType=World, assignmentGameplayTags {gameplayTags [{tagName=NPC.Assignment.Escort}]}}]
    }
   }
   EnvQueryTest_Dot_1 {
    class=EnvQueryTest_Dot
    properties {
     clampMaxType=None
     clampMinType=None
     scoreClampMax=0
     scoreClampMin=0
     scoringFactor=0.1
     testComment=Small DOT to prefer things ahead. Intended for Goals On Circle. Which is why the value is small.
    }
   }
   FortQueryTest_GoalDistance_4 {
    class=FortQueryTest_GoalDistance
    properties {
     bScoreEncounterGoals=false
     bScoreEnemies=false
     bScoreSpecificAssignments=true
     bScoreWorldGoals=false
     clampMaxType=None
     clampMinType=None
     distanceTo=/Script/FortniteAIServer.FortQueryContext_NearbyFriends
     filterType=Minimum
     floatValueMax=0
     floatValueMin=600
     scoreClampMax=0
     scoreClampMin=0
     scoringEquation=Linear
     scoringFactor=1
     testComment=Goals On Circle - Filter out things near my friends to break up clumps of AI.
     testPurpose=Filter
     assignmentIds [{assignmentTeam=Monster, assignmentType=World, assignmentGameplayTags {gameplayTags [{tagName=NPC.Assignment.Escort}]}}]
    }
   }
   FortQueryTest_GoalDistance_2 {
    class=FortQueryTest_GoalDistance
    properties {
     bScoreEncounterGoals=false
     bScoreEnemies=true
     bScoreSpecificAssignments=false
     bScoreWorldGoals=false
     clampMaxType=None
     clampMinType=None
     distanceTo=/Script/FortniteAIServer.FortQueryContext_Goal_SpawnLocation
     filterType=Maximum
     floatValueMax=15000
     scoreClampMax=0
     scoreClampMin=0
     scoringEquation=Linear
     scoringFactor=1
     testComment=Quick/Dirty leashing of AI to delivery area to stop players from kiting AI too far.
     testPurpose=Filter
     assignmentIds []
    }
   }
  ]
 }]
