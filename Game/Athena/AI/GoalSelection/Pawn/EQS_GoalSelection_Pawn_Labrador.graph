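# EQS goal-selection query that picks which perceived enemy pawn the Labrador NPC reacts to.
# A composite generator gathers enemies, then a perception test scores them by per-sense weights, including a small hearing weight, and sense age.
# A gameplay tag test filters out pawns tagged Wildlife or Labrador, noted as keeping wildlife from waking it.
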EnvQuery /Game/Athena/AI/GoalSelection/Pawn/EQS_GoalSelection_Pawn_Labrador
options [{
  generator {
   name=EnvQueryGenerator_Composite_2
   class=EnvQueryGenerator_Composite
   properties {
    optionName=Perceived All Enemies
    generators [&self:EnvQueryGenerator_Composite_2.FortQueryGenerator_Enemies_1]
   }
  }
  tests [
   FortQueryTest_PerceptionAll_2 {
    class=FortQueryTest_PerceptionAll
    properties {
     clampMaxType=SpecifiedValue
     clampMinType=SpecifiedValue
     maxSenseAge=21
     minSenseAge=20
     scoreClampMax=1
     scoringFactor=1
     additionalSenseScores {&Class'&AthenaAISense_Hearing'=0.1}
     senseScores [0,0.1,1,0,0,0]
    }
   }
   FortQueryTest_GoalGameplayTags_0 {
    class=FortQueryTest_GoalGameplayTags
    properties {
     bDefineReferenceValue=false
     bRequireAllProvidedTagQueriesPass=false
     bScoreEncounterGoals=true
     bScoreEnemies=true
     bScoreSpecificAssignments=false
     bScoreWorldGoals=true
     bShouldLookupQueryByTag=false
     bShouldPassWhenQueryNotFound=true
     boolValue=true
     clampMaxType=None
     clampMinType=None
     filterType=Match
     floatValueMax=0
     floatValueMin=0
     multipleContextFilterOp=AllPass
     multipleContextScoreOp=AverageScore
     normalizationType=Absolute
     referenceValue=0
     scoreClampMax=0
     scoreClampMin=0
     scoringEquation=Constant
     scoringFactor=1
     testComment=Don't want be awake by wildlife
     testPurpose=FilterAndScore
     assignmentIds []
     goalActorTagQuery {autoDescription="", tokenStreamVersion=0, userDescription="", queryTokenStream [], tagDictionary []}
     pawnOwnerTagQuery {autoDescription="", tokenStreamVersion=0, userDescription="", queryTokenStream [], tagDictionary []}
     prohibitedAssignmentIds []
     queryLookupTag {tagName=None}
     tagQueryToMatch {
      autoDescription=" NONE( Pawn.Athena.NPC.Wildlife, Pawn.Athena.NPC.Labrador )"
      tokenStreamVersion=0
      userDescription=""
      queryTokenStream [0,1,3,2,0,1]
      tagDictionary [{tagName=Pawn.Athena.NPC.Wildlife}, {tagName=Pawn.Athena.NPC.Labrador}]
     }
    }
   }
  ]
 }]
