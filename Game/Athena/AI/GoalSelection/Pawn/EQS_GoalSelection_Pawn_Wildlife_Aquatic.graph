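# EQS query that selects goals for aquatic wildlife pawns in Athena, choosing among enemies, loot and specific assignment goals.
# A composite generator feeds one option, where tag lookup filters foraged-item assignments and distance tests favor nearer goals, capping range for lures and loot.
# A perception test weights candidates by sense, including hearing.
# World goals must also project onto navmesh through a water-only nav filter.
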EnvQuery /Game/Athena/AI/GoalSelection/Pawn/EQS_GoalSelection_Pawn_Wildlife_Aquatic
options [{
  generator {
   name=EnvQueryGenerator_Composite_0
   class=EnvQueryGenerator_Composite
   properties {
    optionName="Perceived Enemies && Loot"
    generators [
     &self:EnvQueryGenerator_Composite_0.FortQueryGenerator_Enemies_1
     &self:EnvQueryGenerator_Composite_0.FortQueryGenerator_LootGoalsAthena_0
     &self:EnvQueryGenerator_Composite_0.FortQueryGenerator_SpecificAssignmentGoals_0
    ]
   }
  }
  tests [
   FortQueryTest_GoalGameplayTags_1 {
    class=FortQueryTest_GoalGameplayTags
    properties {
     bDefineReferenceValue=false
     bRequireAllProvidedTagQueriesPass=false
     bScoreEncounterGoals=false
     bScoreEnemies=false
     bScoreSpecificAssignments=true
     bScoreWorldGoals=false
     bShouldLookupQueryByTag=true
     bShouldPassWhenQueryNotFound=true
     boolValue=true
     clampMaxType=None
     clampMinType=None
     filterType=Match
     floatValueMax=0
     floatValueMin=0
     multipleContextFilterOp=AllPass
     multipleContextScoreOp=AverageScore
     normalizationType=Absolute
     referenceValue=0
     scoreClampMax=0
     scoreClampMin=0
     scoringEquation=Constant
     scoringFactor=1
     testComment=Find Foraged Items
     testPurpose=Filter
     assignmentIds [{assignmentTeam=Monster, assignmentType=World, assignmentGameplayTags {gameplayTags [{tagName=NPC.Assignment.Loot}]}}]
     goalActorTagQuery {autoDescription="", tokenStreamVersion=0, userDescription="", queryTokenStream [], tagDictionary []}
     pawnOwnerTagQuery {autoDescription="", tokenStreamVersion=0, userDescription="", queryTokenStream [], tagDictionary []}
     prohibitedAssignmentIds []
     queryLookupTag {tagName=NPC.Assignment.Loot}
     tagQueryToMatch {autoDescription="", tokenStreamVersion=0, userDescription="", queryTokenStream [], tagDictionary []}
    }
   }
   FortQueryTest_GoalDistance_1 {
    class=FortQueryTest_GoalDistance
    properties {
     bScoreEncounterGoals=false
     bScoreEnemies=false
     bScoreSpecificAssignments=false
     bScoreWorldGoals=true
     clampMaxType=SpecifiedValue
     clampMinType=SpecifiedValue
     filterType=Maximum
     floatValueMax=10000
     scoreClampMax=5000
     scoreClampMin=200
     scoringEquation=InverseLinear
     scoringFactor=1
     testComment=Lure and Loot Scoring and Filtering
     testPurpose=FilterAndScore
     assignmentIds [{assignmentTeam=Monster, assignmentType=Encounter, assignmentGameplayTags {gameplayTags []}}]
    }
   }
   FortQueryTest_GoalDistance_0 {
    class=FortQueryTest_GoalDistance
    properties {
     bScoreEncounterGoals=true
     bScoreEnemies=true
     bScoreSpecificAssignments=false
     bScoreWorldGoals=false
     clampMaxType=SpecifiedValue
     clampMinType=SpecifiedValue
     filterType=Range
     floatValueMax=0
     scoreClampMax=5000
     scoreClampMin=200
     scoringEquation=InverseLinear
     scoringFactor=0.5
     testComment=Prefer closer things.
     testPurpose=Score
     assignmentIds []
    }
   }
   FortQueryTest_PerceptionAll_0 {
    class=FortQueryTest_PerceptionAll
    properties {
     clampMaxType=SpecifiedValue
     clampMinType=SpecifiedValue
     maxSenseAge=21
     minSenseAge=20
     scoreClampMax=1
     scoringFactor=1
     additionalSenseScores {&Class'&AthenaAISense_Hearing'=0.1}
     senseScores [0,0.1,1,0,0,0]
    }
   }
   FortQueryTest_GoalProject_0 {
    class=FortQueryTest_GoalProject
    properties {
     bDefineReferenceValue=false
     bScoreEncounterGoals=false
     bScoreEnemies=false
     bScoreSpecificAssignments=false
     bScoreWorldGoals=true
     boolValue=true
     clampMaxType=None
     clampMinType=None
     filterType=Match
     floatValueMax=0
     floatValueMin=0
     multipleContextFilterOp=AllPass
     multipleContextScoreOp=AverageScore
     normalizationType=Absolute
     referenceValue=0
     scoreClampMax=0
     scoreClampMin=0
     scoringEquation=Constant
     scoringFactor=1
     testComment=""
     testPurpose=Filter
     assignmentIds []
     goalActorTagQuery {autoDescription="", tokenStreamVersion=0, userDescription="", queryTokenStream [], tagDictionary []}
     prohibitedAssignmentIds []
     projectionData {
      bCanDisableTrace=false
      bCanProjectDown=true
      bCanTraceOnGeometry=false
      bCanTraceOnNavMesh=true
      bOnlyBlockingHits=true
      bTraceComplex=false
      extentX=0
      extentY=10
      extentZ=10
      navigationFilter=/Game/Athena/AI/NavFilters/NavFilter_NPC_Water_ExcludeOthers.NavFilter_NPC_Water_ExcludeOthers_C
      postProjectionVerticalOffset=0
      projectDown=256
      projectUp=256
      serializedChannel=ECC_Visibility
      traceChannel=TraceTypeQuery1
      traceMode=Navigation
      traceProfileName=None
      traceShape=Line
      versionNum=1
     }
    }
   }
  ]
 }]
